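# Looping stereo MetaSound source for vehicle tire audio, mixing surface roll, bare-rim scrape and skid layers.
# Surface loop starts on play with random wave, offset and pitch; rim loop follows Tires changes and IsBareRim.
# Speed and Tires scale all layers, inverted Tires for rim, while smoothed SurfaceMagnitude boosts surface and rim.
# Skid triggers drive an interpolated fade that stops playback below a threshold; the Mobile page drops the skid-start layer.
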MetaSoundSource /VehicleAudio/Surfaces/MSS_Vehicle_Surfaces
inputs [
 BareRim.IsBareRim {type=Bool}
 BareRim.Wave {type=WaveAsset}
 OnStop {type=Trigger}
 Skid {type=Trigger}
 Skid.Loop.Volume {type=Float}
 Skid.Loop.Waves {type=WaveAsset:Array}
 Skid.Start.Volume {type=Float}
 Skid.Start.Waves {type=WaveAsset:Array}
 Speed {type=Float}
 StopSkid {type=Trigger}
 Surface.LoopVolume {type=Float}
 Surface.RimScrapeVolume {type=Float}
 Surface.Waves {type=WaveAsset:Array}
 SurfaceMagnitude {type=Float}
 Tires {type=Float}
 UE.Source.OnPlay {type=Trigger}
]
interfaces [UE.OutputFormat.Stereo:1.0, UE.Source:1.1]
outputs [
 "UE.OutputFormat.Stereo.Audio:0" {type=Audio}
 "UE.OutputFormat.Stereo.Audio:1" {type=Audio}
]
pages [
 Default {
  is_build_graph=true
  input_defaults {
   BareRim.IsBareRim="<MetaSoundBuilderResult.SUCCEEDED: 0>"
   BareRim.Wave=/VehicleAudio/Assets/Tread/Veh_Valet_Skid_Gen_Bare_Rim_Loop_01
   OnStop=null
   Skid=null
   Skid.Loop.Volume=1
   Skid.Start.Volume=1
   Speed=1
   StopSkid=null
   Surface.LoopVolume=1
   Surface.RimScrapeVolume=1.25
   SurfaceMagnitude=0
   Tires=1
   UE.Source.OnPlay=null
   Skid.Loop.Waves []
   Skid.Start.Waves []
   Surface.Waves [
    /VehicleAudio/Assets/Tread/Veh_Valet_Tires_Asphalt_Loop_01
    /VehicleAudio/Assets/Tread/Veh_Valet_Tires_Asphalt_Wet_Loop_01
    /VehicleAudio/Assets/Tread/Veh_Valet_Tires_Dirt_Loop_01
    /VehicleAudio/Assets/Tread/Veh_Valet_Tires_Grass_Loop_01
   ]
  }
  nodes [
   ArrayRandomGet_0 = Array.Random Get:WaveAsset:Array(Next=^Skid, In Array=^Skid.Loop.Waves)
   GetWaveDuration_1 = "UE.Get Wave Duration: "(Wave=@ArrayRandomGet_0.Value)
   n1 = UE.RandomTime(Next=@ArrayRandomGet_0.On Next, Max=@GetWaveDuration_1.Duration)
   TriggerRouteTriggerRouteFloat2_0 = "TriggerRoute.Trigger Route (Float, 2):Float"(Set 0=^Skid, Set 1=^StopSkid) {also {defaults {Value 0=1}}}
   n2 = UE.InterpTo:Audio(Target=@TriggerRouteTriggerRouteFloat2_0.Value) {also {defaults {Interp Time=0.2}}}
   n3 = UE.TriggerOnThreshold:Float(In=@n2.Value) {also {defaults {Threshold=0.01, Type=1}}}
   WavePlayer_2 = UE.Wave Player:Stereo(
    Wave Asset=@ArrayRandomGet_0.Value
    Pitch Shift=/FortMetasound/Patches/RandomSpreadSimple_StartTimePitch(In=@ArrayRandomGet_0.On Next) {also {defaults {PitchShiftSpread=1}}}.RandomPitchShiftOut
    Play=@n1.On Next
    Start Time=@n1.Value
    Stop=@n3.Out
   ) {also {defaults {Loop="<MetaSoundBuilderResult.SUCCEEDED: 0>"}, version [1,1]}}
   ADSREnvelopeADSREnvelope_0 = ADSR Envelope.ADSR Envelope:Float(Trigger Attack=@WavePlayer_2.On Play) {also {defaults {Attack Time=0.15, Decay Time=0.1, Hard Reset="<MetaSoundBuilderResult.SUCCEEDED: 0>", Release Time=0.25, Sustain Level=1}}}
   ArrayRandomGet_1 = Array.Random Get:WaveAsset:Array(In Array=^Skid.Start.Waves, Next=@WavePlayer_2.On Play)
   WavePlayer_1 = UE.Wave Player:Stereo(Play=@ArrayRandomGet_1.On Next, Wave Asset=@ArrayRandomGet_1.Value, Stop=UE.Reroute(Value=UE.Reroute(Value=@n3.Out).Value).Value) {also {version [1,1]}}
   AudioMixerAudioMixerStereo2_0 = "AudioMixer.Audio Mixer (Stereo, 2)"(
    Gain 0=^Skid.Loop.Volume
    Gain 1=^Skid.Start.Volume
    In 0 L=UE.Multiply:Audio by Float(AdditionalOperands=@ADSREnvelopeADSREnvelope_0.Out Envelope, PrimaryOperand=@WavePlayer_2.Out Left).Out
    In 0 R=UE.Multiply:Audio by Float(AdditionalOperands=@ADSREnvelopeADSREnvelope_0.Out Envelope, PrimaryOperand=@WavePlayer_2.Out Right).Out
    In 1 L=@WavePlayer_1.Out Left
    In 1 R=@WavePlayer_1.Out Right
   )
   ArrayRandomGet_2 = Array.Random Get:WaveAsset:Array(In Array=^Surface.Waves, Next=^UE.Source.OnPlay)
   GetWaveDuration_0 = "UE.Get Wave Duration: "(Wave=@ArrayRandomGet_2.Value)
   n4 = UE.RandomTime(Next=@ArrayRandomGet_2.On Next, Max=@GetWaveDuration_0.Duration)
   n5 = UE.RandomFloat(Next=@n4.On Next) {also {defaults {Min=-1}}}
   WavePlayer_3 = UE.Wave Player:Stereo(Wave Asset=@ArrayRandomGet_2.Value, Play=@n5.On Next, Pitch Shift=@n5.Value, Start Time=@n4.Value) {also {defaults {Loop="<MetaSoundBuilderResult.SUCCEEDED: 0>"}, version [1,1]}}
   ADSREnvelopeADSREnvelope_1 = ADSR Envelope.ADSR Envelope:Float(Trigger Attack=@WavePlayer_3.On Play) {also {defaults {Attack Time=0.25, Decay Time=0.1, Release Time=0.25, Sustain Level=1}}}
   TriggerOnValueChange_1 = UE.Trigger On Value Change:Float(Value=^Tires)
   n6 = TriggerCompare.Clamp:Float(A=UE.Reroute(Value=^Tires).Value, Compare=@TriggerOnValueChange_1.Trigger) {also {defaults {B=1, Type=2}}}
   TriggerRouteTriggerRouteBool2_0 = "TriggerRoute.Trigger Route (Bool, 2):Bool"(Set 1=@n6.False, Set 0=@n6.True) {also {defaults {Value 0="<MetaSoundBuilderResult.SUCCEEDED: 0>"}}}
   n7 = VariableAccessor.Bool(Variable=VariableMutator.Bool(Variable=InitVariable.Bool().Variable, Value=@TriggerRouteTriggerRouteBool2_0.Value).Variable)
   TriggerOnValueChange_0 = UE.Trigger On Value Change:Bool(Value=@n7.Value)
   TriggerAnyTriggerAny2_0 = "TriggerAny.Trigger Any (2)"(In 0=^UE.Source.OnPlay, In 1=@TriggerOnValueChange_0.Trigger)
   n8 = TriggerCompare.Clamp:Bool(A=UE.Reroute(Value=@n7.Value).Value, Compare=@TriggerAnyTriggerAny2_0.Out) {also {defaults {B="<MetaSoundBuilderResult.SUCCEEDED: 0>"}}}
   TriggerOnce_0 = UE.Trigger Once(Reset=@n8.False, Trigger In=@n8.True)
   n9 = TriggerCompare.Clamp:Bool(A=^BareRim.IsBareRim, Compare=@TriggerOnce_0.Trigger Out) {also {defaults {B="<MetaSoundBuilderResult.SUCCEEDED: 0>"}}}
   TriggerAnyTriggerAny3_0 = "TriggerAny.Trigger Any (3)"(In 1=UE.Reroute(Value=@n8.False).Value, In 0=@n9.False)
   WavePlayer_0 = UE.Wave Player:Stereo(Wave Asset=^BareRim.Wave, Stop=@TriggerAnyTriggerAny3_0.Out, Play=@n9.True) {also {defaults {Loop="<MetaSoundBuilderResult.SUCCEEDED: 0>"}, version [1,1]}}
   ADSREnvelopeADSREnvelope_2 = ADSR Envelope.ADSR Envelope:Float(Trigger Release=@TriggerAnyTriggerAny3_0.Out, Trigger Attack=@WavePlayer_0.On Play) {also {defaults {Attack Time=0.25, Decay Time=0.1, Release Time=0.1, Sustain Level=0.2}}}
   n10 = VariableMutator.Float(Variable=InitVariable.Float().Variable, Value=UE.Multiply:Float(AdditionalOperands=^Speed, PrimaryOperand=@n2.Value).Out)
   n11 = InitVariable.Float()
   n12 = /FortMetasound/Patches/InvertFloat(FloatIn=^Tires)
   n13 = MapRange.MapRange:Float(In=^Speed) {also {defaults {Out Range B=1.5}}}
   n14 = MapRange.MapRange:Float(In=^SurfaceMagnitude) {also {defaults {In Range A=0.15, In Range B=0.4, Out Range A=1, Out Range B=1.5}}}
   n15 = VariableAccessor.Float(Variable=VariableMutator.Float(Variable=InitVariable.Float().Variable, Value=UE.InterpTo:Audio(Target=@n14.Out Value).Value).Variable)
   n16 = UE.Multiply:Float(
    AdditionalOperands=^Surface.RimScrapeVolume
    PrimaryOperand=UE.Multiply:Float(
     PrimaryOperand=UE.Multiply:Float(PrimaryOperand=@n12.FloatOut, AdditionalOperands=@n13.Out Value).Out
     AdditionalOperands=VariableAccessor.Float(Variable=@n15.Variable).Value
    ).Out
   )
   n17 = InitVariable.Float()
   n18 = MapRange.MapRange:Float(In=^Tires) {also {defaults {Out Range A=0.2}}}
   n19 = UE.Multiply:Float(
    AdditionalOperands=^Surface.LoopVolume
    PrimaryOperand=UE.Multiply:Float(PrimaryOperand=UE.Multiply:Float(PrimaryOperand=^Speed, AdditionalOperands=@n18.Out Value).Out, AdditionalOperands=@n15.Value).Out
   )
   AudioMixerAudioMixerStereo3_0 = "AudioMixer.Audio Mixer (Stereo, 3)"(
    In 2 L=@AudioMixerAudioMixerStereo2_0.Out L
    In 2 R=@AudioMixerAudioMixerStereo2_0.Out R
    In 0 R=UE.Multiply:Audio by Float(AdditionalOperands=@ADSREnvelopeADSREnvelope_1.Out Envelope, PrimaryOperand=@WavePlayer_3.Out Right).Out
    In 0 L=UE.Multiply:Audio by Float(AdditionalOperands=@ADSREnvelopeADSREnvelope_1.Out Envelope, PrimaryOperand=@WavePlayer_3.Out Left).Out
    In 1 L=UE.Multiply:Audio by Float(AdditionalOperands=@ADSREnvelopeADSREnvelope_2.Out Envelope, PrimaryOperand=@WavePlayer_0.Out Left).Out
    In 1 R=UE.Multiply:Audio by Float(AdditionalOperands=@ADSREnvelopeADSREnvelope_2.Out Envelope, PrimaryOperand=@WavePlayer_0.Out Right).Out
    Gain 2=UE.Multiply:Float(AdditionalOperands=^Tires, PrimaryOperand=VariableAccessor.Float(Variable=@n10.Variable).Value).Out
    Gain 1=VariableAccessor.Float(Variable=VariableMutator.Float(Variable=@n11.Variable, Value=@n16.Out).Variable).Value
    Gain 0=VariableAccessor.Float(Variable=VariableMutator.Float(Variable=@n17.Variable, Value=@n19.Out).Variable).Value
   )
  ]
  links [
   {dst="@UE.OutputFormat.Stereo.Audio:0", src=AudioMixerAudioMixerStereo3_0.Out L}
   {dst="@UE.OutputFormat.Stereo.Audio:1", src=AudioMixerAudioMixerStereo3_0.Out R}
  ]
 }
 Mobile {
  nodes [
   ArrayRandomGet_0 = Array.Random Get:WaveAsset:Array(Next=^Skid, In Array=^Skid.Loop.Waves)
   GetWaveDuration_1 = "UE.Get Wave Duration: "(Wave=@ArrayRandomGet_0.Value)
   n1 = UE.RandomTime(Next=@ArrayRandomGet_0.On Next, Max=@GetWaveDuration_1.Duration)
   TriggerRouteTriggerRouteFloat2_0 = "TriggerRoute.Trigger Route (Float, 2):Float"(Set 0=^Skid, Set 1=^StopSkid) {also {defaults {Value 0=1}}}
   n2 = UE.InterpTo:Audio(Target=@TriggerRouteTriggerRouteFloat2_0.Value) {also {defaults {Interp Time=0.2}}}
   WavePlayer_1 = UE.Wave Player:Stereo(
    Wave Asset=@ArrayRandomGet_0.Value
    Pitch Shift=&RandomSpreadSimple_StartTimePitch(In=@ArrayRandomGet_0.On Next) {also {defaults {PitchShiftSpread=1}}}.RandomPitchShiftOut
    Play=@n1.On Next
    Start Time=@n1.Value
    Stop=UE.TriggerOnThreshold:Float(In=@n2.Value) {also {defaults {Threshold=0.01, Type=1}}}.Out
   ) {also {defaults {Loop="<MetaSoundBuilderResult.SUCCEEDED: 0>"}, version [1,1]}}
   ADSREnvelopeADSREnvelope_0 = ADSR Envelope.ADSR Envelope:Float(Trigger Attack=@WavePlayer_1.On Play) {also {defaults {Attack Time=0.15, Decay Time=0.1, Hard Reset="<MetaSoundBuilderResult.SUCCEEDED: 0>", Release Time=0.25, Sustain Level=1}}}
   n3 = VariableMutator.Float(Variable=InitVariable.Float().Variable, Value=UE.Multiply:Float(AdditionalOperands=^Speed, PrimaryOperand=@n2.Value).Out)
   ArrayRandomGet_1 = Array.Random Get:WaveAsset:Array(In Array=^Surface.Waves, Next=^UE.Source.OnPlay)
   GetWaveDuration_0 = "UE.Get Wave Duration: "(Wave=@ArrayRandomGet_1.Value)
   n4 = UE.RandomTime(Next=@ArrayRandomGet_1.On Next, Max=@GetWaveDuration_0.Duration)
   n5 = UE.RandomFloat(Next=@n4.On Next) {also {defaults {Min=-1}}}
   WavePlayer_2 = UE.Wave Player:Stereo(Wave Asset=@ArrayRandomGet_1.Value, Play=@n5.On Next, Pitch Shift=@n5.Value, Start Time=@n4.Value) {also {defaults {Loop="<MetaSoundBuilderResult.SUCCEEDED: 0>"}, version [1,1]}}
   ADSREnvelopeADSREnvelope_1 = ADSR Envelope.ADSR Envelope:Float(Trigger Attack=@WavePlayer_2.On Play) {also {defaults {Attack Time=0.25, Decay Time=0.1, Release Time=0.25, Sustain Level=1}}}
   TriggerOnValueChange_1 = UE.Trigger On Value Change:Float(Value=^Tires)
   n6 = TriggerCompare.Clamp:Float(A=UE.Reroute(Value=^Tires).Value, Compare=@TriggerOnValueChange_1.Trigger) {also {defaults {B=1, Type=2}}}
   TriggerRouteTriggerRouteBool2_0 = "TriggerRoute.Trigger Route (Bool, 2):Bool"(Set 1=@n6.False, Set 0=@n6.True) {also {defaults {Value 0="<MetaSoundBuilderResult.SUCCEEDED: 0>"}}}
   n7 = VariableAccessor.Bool(Variable=VariableMutator.Bool(Variable=InitVariable.Bool().Variable, Value=@TriggerRouteTriggerRouteBool2_0.Value).Variable)
   TriggerOnValueChange_0 = UE.Trigger On Value Change:Bool(Value=@n7.Value)
   TriggerAnyTriggerAny2_0 = "TriggerAny.Trigger Any (2)"(In 0=^UE.Source.OnPlay, In 1=@TriggerOnValueChange_0.Trigger)
   n8 = TriggerCompare.Clamp:Bool(A=UE.Reroute(Value=@n7.Value).Value, Compare=@TriggerAnyTriggerAny2_0.Out) {also {defaults {B="<MetaSoundBuilderResult.SUCCEEDED: 0>"}}}
   TriggerOnce_0 = UE.Trigger Once(Reset=@n8.False, Trigger In=@n8.True)
   n9 = TriggerCompare.Clamp:Bool(A=^BareRim.IsBareRim, Compare=@TriggerOnce_0.Trigger Out) {also {defaults {B="<MetaSoundBuilderResult.SUCCEEDED: 0>"}}}
   TriggerAnyTriggerAny3_0 = "TriggerAny.Trigger Any (3)"(In 1=UE.Reroute(Value=@n8.False).Value, In 0=@n9.False)
   WavePlayer_0 = UE.Wave Player:Stereo(Wave Asset=^BareRim.Wave, Stop=@TriggerAnyTriggerAny3_0.Out, Play=@n9.True) {also {defaults {Loop="<MetaSoundBuilderResult.SUCCEEDED: 0>"}, version [1,1]}}
   ADSREnvelopeADSREnvelope_2 = ADSR Envelope.ADSR Envelope:Float(Trigger Release=@TriggerAnyTriggerAny3_0.Out, Trigger Attack=@WavePlayer_0.On Play) {also {defaults {Attack Time=0.25, Decay Time=0.1, Release Time=0.1, Sustain Level=0.2}}}
   n10 = InitVariable.Float()
   n11 = &InvertFloat(FloatIn=^Tires)
   n12 = MapRange.MapRange:Float(In=^Speed) {also {defaults {Out Range B=1.5}}}
   n13 = MapRange.MapRange:Float(In=^SurfaceMagnitude) {also {defaults {In Range A=0.15, In Range B=0.4, Out Range A=1, Out Range B=1.5}}}
   n14 = VariableAccessor.Float(Variable=VariableMutator.Float(Variable=InitVariable.Float().Variable, Value=UE.InterpTo:Audio(Target=@n13.Out Value).Value).Variable)
   n15 = UE.Multiply:Float(
    AdditionalOperands=^Surface.RimScrapeVolume
    PrimaryOperand=UE.Multiply:Float(
     PrimaryOperand=UE.Multiply:Float(PrimaryOperand=@n11.FloatOut, AdditionalOperands=@n12.Out Value).Out
     AdditionalOperands=VariableAccessor.Float(Variable=@n14.Variable).Value
    ).Out
   )
   n16 = InitVariable.Float()
   n17 = MapRange.MapRange:Float(In=^Tires) {also {defaults {Out Range A=0.2}}}
   n18 = UE.Multiply:Float(
    AdditionalOperands=^Surface.LoopVolume
    PrimaryOperand=UE.Multiply:Float(PrimaryOperand=UE.Multiply:Float(PrimaryOperand=^Speed, AdditionalOperands=@n17.Out Value).Out, AdditionalOperands=@n14.Value).Out
   )
   AudioMixerAudioMixerStereo3_0 = "AudioMixer.Audio Mixer (Stereo, 3)"(
    In 2 L=UE.Multiply:Audio by Float(
     AdditionalOperands=^Skid.Loop.Volume
     PrimaryOperand=UE.Multiply:Audio by Float(AdditionalOperands=@ADSREnvelopeADSREnvelope_0.Out Envelope, PrimaryOperand=@WavePlayer_1.Out Left).Out
    ).Out
    Gain 2=UE.Multiply:Float(AdditionalOperands=^Tires, PrimaryOperand=VariableAccessor.Float(Variable=@n3.Variable).Value).Out
    In 2 R=UE.Multiply:Audio by Float(
     AdditionalOperands=^Skid.Loop.Volume
     PrimaryOperand=UE.Multiply:Audio by Float(AdditionalOperands=@ADSREnvelopeADSREnvelope_0.Out Envelope, PrimaryOperand=@WavePlayer_1.Out Right).Out
    ).Out
    In 0 R=UE.Multiply:Audio by Float(AdditionalOperands=@ADSREnvelopeADSREnvelope_1.Out Envelope, PrimaryOperand=@WavePlayer_2.Out Right).Out
    In 0 L=UE.Multiply:Audio by Float(AdditionalOperands=@ADSREnvelopeADSREnvelope_1.Out Envelope, PrimaryOperand=@WavePlayer_2.Out Left).Out
    In 1 L=UE.Multiply:Audio by Float(AdditionalOperands=@ADSREnvelopeADSREnvelope_2.Out Envelope, PrimaryOperand=@WavePlayer_0.Out Left).Out
    In 1 R=UE.Multiply:Audio by Float(AdditionalOperands=@ADSREnvelopeADSREnvelope_2.Out Envelope, PrimaryOperand=@WavePlayer_0.Out Right).Out
    Gain 1=VariableAccessor.Float(Variable=VariableMutator.Float(Variable=@n10.Variable, Value=@n15.Out).Variable).Value
    Gain 0=VariableAccessor.Float(Variable=VariableMutator.Float(Variable=@n16.Variable, Value=@n18.Out).Variable).Value
   )
   InitVariable.WaveAsset()
  ]
  links [
   {dst="@UE.OutputFormat.Stereo.Audio:0", src=AudioMixerAudioMixerStereo3_0.Out L}
   {dst="@UE.OutputFormat.Stereo.Audio:1", src=AudioMixerAudioMixerStereo3_0.Out R}
  ]
 }
]
properties {
 assetImportData=&self:AssetImportData
 attenuationSettings=/VehicleAudio/Settings/Vehicle_Movement_Close_Detail_Att
 bLooping=true
 duration=10000
 maxDistance=12000
 numChannels=2
 outputFormat=Stereo
 sampleRate=48000
 soundClassObject=/Game/Sounds/FortSoundClassesAndModes/Fort_Vehicles_Tires
 concurrencyOverrides {maxCount=4, voiceStealReleaseTime=0.1}
 concurrencySet [&/Conc_Vehicle_Surface]
}
