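# MetaSound Source preset of MSS_Vehicle_Surfaces for the mud surface variant of big off-road tire vehicle audio, a looping stereo tire and skid sound.
# The inherited graph exposes speed, surface magnitude, tire, skid and bare-rim inputs.
# This preset supplies a mud skid loop, no skid-start waves, a bare-rim scrape loop, and asphalt, wet asphalt, dirt and grass tread loops.
# It uses close-detail vehicle attenuation and the exterior vehicle sound class.
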MetaSoundSource /VehicleAudio/Surfaces/TireMods/OffRoadTires/MSS_Vehicle_Mod_BigTire_Surfaces_Mud
preset_parent=/VehicleAudio/Surfaces/MSS_Vehicle_Surfaces
inputs [
 BareRim.IsBareRim {type=Bool}
 BareRim.Wave {type=WaveAsset}
 OnStop {type=Trigger}
 Skid {type=Trigger}
 Skid.Loop.Volume {type=Float}
 Skid.Loop.Waves {type=WaveAsset:Array}
 Skid.Start.Volume {type=Float}
 Skid.Start.Waves {type=WaveAsset:Array}
 Speed {type=Float}
 StopSkid {type=Trigger}
 Surface.LoopVolume {type=Float}
 Surface.RimScrapeVolume {type=Float}
 Surface.Waves {type=WaveAsset:Array}
 SurfaceMagnitude {type=Float}
 Tires {type=Float}
 UE.Source.OnPlay {type=Trigger}
]
interfaces [UE.OutputFormat.Stereo:1.0, UE.Source:1.1]
outputs [
 "UE.OutputFormat.Stereo.Audio:0" {type=Audio}
 "UE.OutputFormat.Stereo.Audio:1" {type=Audio}
]
pages [Default {input_defaults {
   BareRim.IsBareRim="<MetaSoundBuilderResult.SUCCEEDED: 0>"
   BareRim.Wave=/VehicleAudio/Assets/Tread/Veh_Valet_Skid_Gen_Bare_Rim_Loop_01
   OnStop=null
   Skid=null
   Skid.Loop.Volume=1
   Skid.Start.Volume=1
   Speed=1
   StopSkid=null
   Surface.LoopVolume=0.8
   Surface.RimScrapeVolume=1.25
   SurfaceMagnitude=0
   Tires=1
   UE.Source.OnPlay=null
   Skid.Loop.Waves [/VehicleAudio/Assets/Skid/Veh_Valet_Skid_Mud_Loop]
   Skid.Start.Waves []
   Surface.Waves [
    /VehicleAudio/Assets/Tread/Veh_Valet_Tires_Asphalt_Loop_01
    /VehicleAudio/Assets/Tread/Veh_Valet_Tires_Asphalt_Wet_Loop_01
    /VehicleAudio/Assets/Tread/Veh_Valet_Tires_Dirt_Loop_01
    /VehicleAudio/Assets/Tread/Veh_Valet_Tires_Grass_Loop_01
   ]
  }}]
properties {
 assetImportData=&self:AssetImportData
 attenuationSettings=/VehicleAudio/Settings/Vehicle_Movement_Close_Detail_Att
 bLooping=true
 duration=10000
 maxDistance=12000
 numChannels=2
 outputFormat=Stereo
 sampleRate=48000
 soundClassObject=&Fort_Vehicles_Exterior
}
